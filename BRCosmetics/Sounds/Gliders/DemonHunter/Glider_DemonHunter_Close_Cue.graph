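# Sound cue that plays the closing sound for the DemonHunter cosmetic glider.
# A random node picks between two equally weighted close wave variations.
# Playback uses shared glider open/close attenuation and concurrency, the glider land sound class, and a common glider-close audio property sheet, at slightly reduced volume.
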SoundCue /BRCosmetics/Sounds/Gliders/DemonHunter/Glider_DemonHunter_Close_Cue
root=Random_0
nodes [
 {id=Random_0, children [WavePlayer_1,WavePlayer_2], properties {weights [1,1]}}
 {id=WavePlayer_1, properties {soundWaveAssetPtr=&/SFX_Glider_DemonHunter_Close_01}}
 {id=WavePlayer_2, properties {soundWaveAssetPtr=&/SFX_Glider_DemonHunter_Close_02}}
]
properties {
 attenuationSettings=/Game/Sounds/Attenuation/Fort_Player_Glider_OpenClose_Athena_Att
 audioPropertiesSheet=/BRCosmetics/Sounds/Settings/PropertySheets/Glider/APS_Cosmetic_Glider_Close
 maxDistance=10000
 soundClassObject=&Fort_Gliders_Land
 volumeMultiplier=0.7
 concurrencySet [&Athena_GliderOpenClose_Conc]
}
