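# Sound cue for the opening of the DemonHunter cosmetic glider.
# A mixer layers the main open sound with one giggle picked at random, with equal weight, from three variations.
# Playback uses shared glider open/close attenuation and concurrency, a glider-open audio property sheet, the gliders sound class and a reduced volume multiplier.
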SoundCue /BRCosmetics/Sounds/Gliders/DemonHunter/Glider_DemonHunter_Open_Cue
root=Mixer_0
nodes [
 {id=Mixer_0, children [WavePlayer_1,Random_0], properties {inputVolume [1,1]}}
 {id=Random_0, children [WavePlayer_0,WavePlayer_2,WavePlayer_3], properties {weights [1,1,1]}}
 {id=WavePlayer_0, properties {soundWaveAssetPtr=&/SFX_Glider_DemonHunter_Open_Giggle_01}}
 {id=WavePlayer_1, properties {soundWaveAssetPtr=&/SFX_Glider_DemonHunter_Open}}
 {id=WavePlayer_2, properties {soundWaveAssetPtr=&/SFX_Glider_DemonHunter_Open_Giggle_02}}
 {id=WavePlayer_3, properties {soundWaveAssetPtr=&/SFX_Glider_DemonHunter_Open_Giggle_03}}
]
properties {
 attenuationSettings=/Game/Sounds/Attenuation/Fort_Player_Glider_OpenClose_Athena_Att
 audioPropertiesSheet=/BRCosmetics/Sounds/Settings/PropertySheets/Glider/APS_Cosmetic_Glider_Open
 maxDistance=10000
 soundClassObject=&Fort_Gliders_Land
 volumeMultiplier=0.7
 concurrencySet [&Athena_GliderOpenClose_Conc]
}
